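# Default demolition impact Sound Cue that plays rock impact one-shots in the Fort_Impacts sound class, with a raised base pitch.
# A Quality Level root picks one of four modulated branches: random among ten waves, random among five, or either of two identical single-wave branches.
# The single-wave branches use a wider pitch range to vary their one sample.
# A shared concurrency rule, named for a four-voice limit stopping furthest or oldest, governs overlapping impacts.
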SoundCue /Clyde/Sounds/Fort_Impact_Sounds/Demolition/Default_Impact_Cue
root=QualityLevel_0
nodes [
 {id=Modulator_0, children [WavePlayer_37], properties {pitchMax=1.18, pitchMin=0.88}}
 {id=Modulator_1, children [Random_3]}
 {id=Modulator_2, children [Random_0]}
 {id=Modulator_3, children [WavePlayer_38], properties {pitchMax=1.18, pitchMin=0.88}}
 {id=QualityLevel_0, children [Modulator_1,Modulator_2,Modulator_0,Modulator_3]}
 {id=Random_0, children [WavePlayer_29, WavePlayer_27, WavePlayer_34, WavePlayer_32, WavePlayer_31], properties {weights [1,1,1,1,1]}}
 {
  id=Random_3
  children [WavePlayer_2,WavePlayer_1,WavePlayer_7,WavePlayer_5,WavePlayer_4,WavePlayer_3,WavePlayer_0,WavePlayer_8,WavePlayer_6,WavePlayer_9]
  properties {weights [1,1,1,1,1,1,1,1,1,1]}
 }
 {id=WavePlayer_0, properties {soundWaveAssetPtr=&/rock_impact_03}}
 {id=WavePlayer_1, properties {soundWaveAssetPtr=&/rock_impact_05}}
 {id=WavePlayer_2, properties {soundWaveAssetPtr=&/rock_impact_04}}
 {id=WavePlayer_27, properties {soundWaveAssetPtr=&rock_impact_05}}
 {id=WavePlayer_29, properties {soundWaveAssetPtr=&rock_impact_04}}
 {id=WavePlayer_3, properties {soundWaveAssetPtr=&/rock_impact_02}}
 {id=WavePlayer_31, properties {soundWaveAssetPtr=&/rock_impact_01}}
 {id=WavePlayer_32, properties {soundWaveAssetPtr=&/rock_impact_08}}
 {id=WavePlayer_34, properties {soundWaveAssetPtr=&/rock_impact_07}}
 {id=WavePlayer_37, properties {soundWaveAssetPtr=&rock_impact_04}}
 {id=WavePlayer_38, properties {soundWaveAssetPtr=&rock_impact_04}}
 {id=WavePlayer_4, properties {soundWaveAssetPtr=&rock_impact_01}}
 {id=WavePlayer_5, properties {soundWaveAssetPtr=&rock_impact_08}}
 {id=WavePlayer_6, properties {soundWaveAssetPtr=&/rock_impact_09}}
 {id=WavePlayer_7, properties {soundWaveAssetPtr=&rock_impact_07}}
 {id=WavePlayer_8, properties {soundWaveAssetPtr=&/rock_impact_06}}
 {id=WavePlayer_9, properties {soundWaveAssetPtr=&/rock_impact_10}}
]
properties {
 maxDistance=2097152
 pitchMultiplier=1.5
 soundClassObject=&Fort_Impacts
 volumeMultiplier=1
 concurrencyOverrides {maxCount=4}
 concurrencySet [/Game/Sounds/Concurrency_Rules/Demo_FourMaximum_StopFurthestOldest_VolumeScale]
}
