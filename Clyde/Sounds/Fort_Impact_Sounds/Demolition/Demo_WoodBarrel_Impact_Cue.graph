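# Sound cue for wooden barrel impacts, filed under demolition impact sounds and routed to the Fort_Impacts sound class.
# A mixer layers random barrel impact waves at full input volume with random generic wood impacts at half volume.
# Each layer picks evenly from three waves through its own pitch modulator, varying the barrel around normal pitch and raising the wood layer.
# A shared object-impact concurrency asset governs simultaneous playback.
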SoundCue /Clyde/Sounds/Fort_Impact_Sounds/Demolition/Demo_WoodBarrel_Impact_Cue
root=Mixer_4
nodes [
 {id=Mixer_4, children [Modulator_2,Modulator_1], properties {inputVolume [0.5,1]}}
 {id=Modulator_1, children [Random_31], properties {pitchMax=1.07, pitchMin=0.92}}
 {id=Modulator_2, children [Random_32], properties {pitchMax=1.5, pitchMin=1.2}}
 {id=Random_31, children [WavePlayer_202, WavePlayer_203, WavePlayer_204], properties {weights [1,1,1]}}
 {id=Random_32, children [WavePlayer_206, WavePlayer_207, WavePlayer_209], properties {weights [1,1,1]}}
 {id=WavePlayer_202, properties {soundWaveAssetPtr=&/WoodBarrel_Impact01}}
 {id=WavePlayer_203, properties {soundWaveAssetPtr=&/WoodBarrel_Impact02}}
 {id=WavePlayer_204, properties {soundWaveAssetPtr=&/WoodBarrel_Impact03}}
 {id=WavePlayer_206, properties {soundWaveAssetPtr=&/wood_impact_06}}
 {id=WavePlayer_207, properties {soundWaveAssetPtr=&/wood_impact_07}}
 {id=WavePlayer_209, properties {soundWaveAssetPtr=&/wood_impact_09}}
]
properties {
 maxDistance=2097152
 soundClassObject=&Fort_Impacts
 volumeMultiplier=0.34
 concurrencyOverrides {resolutionRule=PreventNew}
 concurrencySet [&CONC_Object_Impacts_FourMax]
}
